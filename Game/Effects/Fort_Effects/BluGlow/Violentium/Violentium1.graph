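# Subsurface material for Violentium, a BluGlow effect object with an animated spiky surface and violet glow.
# Time-driven sine masks blend three morph-target rows from position and normal textures, then the result rotates and wobbles with noise through World Position Offset.
# Masked glow, sparkle and rotating zoom-noise light shafts feed base, emissive and subsurface color.
# Static switches add hologram and chrome variants, while LOD and black-hole functions add further vertex offsets.
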Material /Game/Effects/Fort_Effects/BluGlow/Violentium/Violentium1
graph {
 properties {
  bEnableResponsiveAA=true
  bTangentSpaceNormal=false
  ditherOpacityMask=true
  numCustomizedUVs=4
  pixelDepthOffsetMode=PDOM_Legacy
  shadingModel=MSM_Subsurface
  subsurfaceProfile=&/Violentium_Profile
  usedShadingModels=Subsurface
 }
 nodes [
  n1 = MaterialFunctionCall() {materialFunction=/Game/Effects/Fort_Effects/BluGlow/BluglowPositionScale}
  n2 = Time() {bOverride_Period=true, period=999999995904}
  n3 = ScalarParameter() {defaultValue=0.05, parameterName=Random Rotation Speed}
  n4 = Add(
   A=@n1.Random
   B=Add(
    A=Multiply(A=@n2, B=@n3) {constB=0.1}
    B=MaterialFunctionCall() {materialFunction=&BluglowPositionScale}.Random
   )
  )
  n5 = MaterialFunctionCall(
   Hue Shift Percentage=@n4
   Texture=MaterialFunctionCall(Hue Shift Percentage=@n4, Texture=Constant3Vector() {constant=<rgba 1 0 0 0>}) {materialFunction=&HueShift}.Result
  ) {materialFunction=&HueShift}
  n6 = MaterialFunctionCall(Float2=TextureCoordinate() {coordinateIndex=1}) {materialFunction=&BreakOutFloat2Components}
  n7 = Divide(A=Constant() {r=1}, B=Constant() {desc=total, r=4})
  n8 = Multiply(A=@n7) {constB=0.5}
  n9 = Multiply(A=@n7, B=Constant() {desc=index, r=1})
  n10 = AppendVector(A=@n6.R, B=Add(A=@n8, B=@n9))
  n11 = TextureSample(UVs=@n10) {
   samplerSource=SSM_Wrap_WorldGroupSettings
   samplerType=SAMPLERTYPE_LinearColor
   texture=&/ViolentiumAnimation_UV1_Normals
  }
  n12 = Add(A=@n9, B=@n9)
  n13 = AppendVector(A=@n6.R, B=Add(A=@n8, B=@n12))
  n14 = TextureSample(UVs=@n13) {
   samplerSource=SSM_Wrap_WorldGroupSettings
   samplerType=SAMPLERTYPE_LinearColor
   texture=&ViolentiumAnimation_UV1_Normals
  }
  n15 = Desaturation(None=MaterialFunctionCall() {materialFunction=&BluglowPositionScale}.Position, Fraction=Constant() {r=1})
  n16 = Add(
   A=Add(
    A=Divide(A=@n15) {constB=300}
    B=MaterialFunctionCall() {materialFunction=&BoundingBoxBased_0-1_UVW}.RGB
   )
   B=Time() {bOverride_Period=true, period=99999997952}
  )
  n17 = TextureCoordinate() {coordinateIndex=2}
  n18 = Multiply(A=Time() {bOverride_Period=true, period=99999997952}, B=ScalarParameter() {defaultValue=0.0826, parameterName=Panning speed Spke Mask})
  n19 = SphereMask(
   A=MaterialFunctionCall() {materialFunction=&LinearGradient}.VGradient
   B=Constant() {r=0.5}
  ) {attenuationRadius=0.5, hardnessPercent=0}
  n20 = Constant() {r=3.1395}
  n21 = Clamp(None=Multiply(A=Power(Base=@n19, Exp=@n20)) {constB=1.5})
  n22 = LinearInterpolate(
   A=Constant3Vector() {constant=<rgba 0.425 0.2077 0.1307 1>}
   B=TextureSample(UVs=Panner(Coordinate=@n17, Time=@n18) {speedX=0.0338, speedY=-0.0349}) {samplerType=SAMPLERTYPE_LinearColor, texture=&T_NoisePack_LinearColor}.RGB
   Alpha=@n21
  )
  n23 = Add(
   A=Multiply(A=MaterialFunctionCall(Hue Shift Percentage=@n16, Texture=@n22) {materialFunction=&HueShift}.Result) {constB=1.25}
   B=Multiply(
    A=Time() {bOverride_Period=true, period=9999999}
    B=VectorParameter() {parameterName=Morph Target Transition Speeds, defaultValue=<rgba 0.0164 0.094 0.9091 1>}.RGB
   )
  )
  n24 = MaterialFunctionCall(Float3=ConstantBiasScale(None=Sine(None=@n23))) {materialFunction=&BreakOutFloat3Components}
  n25 = Multiply(
   A=Power(Base=Multiply(A=@n24.G, B=@n24.G), Exp=ScalarParameter() {defaultValue=3.8713, parameterName=Spike Power})
   B=ScalarParameter() {defaultValue=1, parameterName=Spike mult}
  )
  n26 = AppendVector(A=@n6.R, B=Add(A=@n8, B=Add(A=@n12, B=@n9)))
  n27 = TextureSample(UVs=@n26) {
   samplerSource=SSM_Wrap_WorldGroupSettings
   samplerType=SAMPLERTYPE_LinearColor
   texture=&ViolentiumAnimation_UV1_Normals
  }
  n28 = ConstantBiasScale(None=LinearInterpolate(A=LinearInterpolate(A=@n11.RGB, B=@n14.RGB, Alpha=@n24.R), B=@n27.RGB, Alpha=@n25)) {bias=-0.5, scale=2}
  n29 = MaterialFunctionCall(
   Rotation Axis=@n5.Result
   Rotation Angle=@n4
   World Space Vertex Normals=Normalize(VectorInput=Transform(None=@n28) {transformSourceType=TRANSFORMSOURCE_Local})
  ) {materialFunction=&FixRotateAboutAxisNormals}
  n30 = MaterialFunctionCall(Float3=@n29.New World Space Vertex Normal) {materialFunction=&BreakOutFloat3Components}
  AppendVector(A=@n30.R, B=@n30.R) {desc=2}
  n31 = TextureSample(UVs=@n10) {
   samplerSource=SSM_Wrap_WorldGroupSettings
   samplerType=SAMPLERTYPE_LinearColor
   texture=&/ViolentiumAnimation_UV1
  }
  n32 = TextureSample(UVs=@n13) {
   samplerSource=SSM_Wrap_WorldGroupSettings
   samplerType=SAMPLERTYPE_LinearColor
   texture=&ViolentiumAnimation_UV1
  }
  n33 = TextureSample(UVs=@n26) {
   samplerSource=SSM_Wrap_WorldGroupSettings
   samplerType=SAMPLERTYPE_LinearColor
   texture=&ViolentiumAnimation_UV1
  }
  n34 = Transform(None=LinearInterpolate(A=LinearInterpolate(A=@n31.RGB, B=@n32.RGB, Alpha=@n24.R), B=@n33.RGB, Alpha=@n25)) {transformSourceType=TRANSFORMSOURCE_Local}
  n35 = Add(A=DotProduct(A=@n34, B=VertexNormalWS()), B=ScalarParameter() {defaultValue=1.8774, parameterName=Spike Grad 1})
  n36 = ScalarParameter() {defaultValue=1.681, parameterName=Spike Grad 2}
  AppendVector(A=@n30.B, B=Clamp(None=Divide(A=@n35, B=@n36) {constB=25})) {desc=3}
  Constant()
  Constant() {r=0.1}
  Constant() {r=0.5}
  Distance(A=@n34, B=Constant())
  MaterialFunctionCall!(Morph Animation, Morph Normal, Morph Texture, Number of Morph Targets, 0-1 Animation Value?, Morph Target UV) {materialFunction=/Engine/Functions/Engine_MaterialFunctions02/MAXScripts/MS_VertexAnimationTools_MorphTargets}
  n37 = Time() {bIgnorePause=true, bOverride_Period=true, period=100000000}
  n38 = Add(A=Multiply(A=Constant2Vector() {g=-0.2278, r=0.1596}, B=@n37), B=TextureCoordinate())
  n39 = Add(A=TextureCoordinate() {uTiling=2, vTiling=2}, B=Multiply(A=@n37, B=Constant2Vector() {g=0.2604, r=-0.1806}))
  Multiply(
   A=LinearInterpolate(
    A=TextureSample(UVs=@n38) {samplerType=SAMPLERTYPE_LinearGrayscale, texture=/Game/Global/Textures/Default/Noise/T_SwirlNoise_N_height}.R
    B=TextureSample(UVs=@n39) {samplerType=SAMPLERTYPE_LinearGrayscale, texture=&T_SwirlNoise_N_height}.R
   )
   B=ScalarParameter() {defaultValue=4.9143, parameterName=Bubble effect}
  )
  n40 = MaterialFunctionCall() {materialFunction=&ObjectLocalBounds}
  n41 = Subtract(
   A=Subtract(
    A=MaterialFunctionCall() {materialFunction=&ComponentPivotLocation}.Result
    B=WorldPosition().XYZ
   )
   B=@n40.Local Bounds Minimum
  )
  n42 = MaterialFunctionCall!(
   Vector 3 Normal Sobel=Constant3Vector()
   Vertex Normal WS=@n29.New World Space Vertex Normal
   World Space Normal=PixelNormalWS()
   Bounding box G=ComponentMask(None=Divide(A=@n41, B=@n40.Local Bounds Size)) {g=true}
  ) {materialFunction=/Game/Effects/Fort_Effects/Gameplay/Pickups/Materials/M_Pickup_Blueprint_Function}
  n43 = TextureSampleParameter2D(UVs=@n38) {group=Wiggle, parameterName=Wiggle Normal_1, samplerType=SAMPLERTYPE_Normal, texture=/Game/Global/Textures/Default/Noise/T_SwirlNoise_N}
  n44 = TextureSampleParameter2D(UVs=@n39) {group=Wiggle, parameterName=Wiggle Normal, samplerType=SAMPLERTYPE_Normal, texture=&T_SwirlNoise_N}
  n45 = LinearInterpolate(A=Constant3Vector() {constant=<rgba 0 0 1 0>}, B=Normalize(VectorInput=Add(A=@n43.RGB, B=@n44.RGB)), Alpha=@n21)
  n46 = ScalarParameter() {defaultValue=-0.4453, parameterName=Base roughness}
  n47 = ScalarParameter() {defaultValue=0.5383, parameterName=Base roughness_1}
  n48 = OneMinus(None=ComponentMask(None=TextureCoordinate() {coordinateIndex=3}) {g=true})
  n49 = Clamp(None=LinearInterpolate(
    A=ScalarParameter() {defaultValue=-0.2141, parameterName=Spike mask a}
    B=ScalarParameter() {defaultValue=0.1464, parameterName=Spike MAsk 2}
    Alpha=@n48
   ))
  n50 = Divide(
   A=MaterialFunctionCall() {materialFunction=&ComponentPivotLocation}.Result
   B=Constant() {r=700}
  )
  n51 = WorldPosition()
  n52 = Add(A=@n34, B=RotateAboutAxis(NormalizedRotationAxis=@n5.Result, RotationAngle=@n4, PivotPoint=@n1.Position, Position=Add(A=@n34, B=@n51.XYZ)))
  n53 = Add(A=WorldPosition() {worldPositionShaderOffset=WPT_ExcludeAllShaderOffsets}.XYZ, B=@n52)
  n54 = MaterialFunctionCall() {materialFunction=&BluglowPositionScale}
  n55 = Add(A=Divide(A=ComponentMask(None=@n53) {b=true}, B=Add(A=@n54.Radius, B=@n54.Radius))) {constB=0.5}
  n56 = Add(
   A=Multiply(A=Time() {bOverride_Period=true, period=100000000376832}, B=ScalarParameter() {defaultValue=0.0266, parameterName=Overall Wiggle Time})
   B=Multiply(A=ScalarParameter() {defaultValue=0.06, parameterName=Overall Wiggle Time Z offset}, B=@n55)
  )
  n57 = TextureSampleParameter2D(UVs=ComponentMask(None=Add(A=@n50, B=@n56)) {g=true, r=true}) {
   group=Wiggle
   parameterName=Wiggle Normal_2
   samplerType=SAMPLERTYPE_LinearColor
   texture=&HighContrastNoise
  }
  n58 = AppendVector(A=Constant2Vector(), B=ScalarParameter() {defaultValue=3, parameterName=Z Stretch})
  n59 = Add(
   A=Multiply(A=Subtract(A=@n57.RGB) {constB=0.5}, B=ScalarParameter() {defaultValue=11.8819, parameterName=Overall Wiggle})
   B=Add(A=Multiply(A=@n55, B=@n58), B=VectorParameter() {parameterName=Model Offset, defaultValue=<rgba 0 0 -10.2044 1>}.RGB)
  )
  n60 = MaterialFunctionCall!(
   Scale=ScalarParameter() {defaultValue=-0.1, parameterName=SparkleScale}
   Sparkle World center point=Add(A=@n59, B=MaterialFunctionCall() {materialFunction=&BluglowPositionScale}.Position)
  ) {materialFunction=/Game/Packages/Fortnite_UniqueMaterials_1/Effects/FortniteSparkle}
  n61 = ScalarParameter() {defaultValue=0.2549, parameterName=Sparkle brightness}
  n62 = Add(B=Add(B=Multiply(A=@n60.Result, B=@n61)) {constA=0.5})
  n63 = Constant()
  n64 = Add(A=@n60.UVS) {constB=0}
  n65 = ScalarParameter() {defaultValue=0.9692, parameterName=InternalGlowRadius}
  n66 = StaticSwitchParameter(
   True=Multiply(A=@n65, B=ParticleRadius())
   False=Multiply(A=MaterialFunctionCall() {materialFunction=&BluglowPositionScale}.Radius, B=@n65)
  ) {group=Particle, parameterName=isParticle}
  n67 = Transform(None=PixelNormalWS()) {transformSourceType=TRANSFORMSOURCE_World, transformType=TRANSFORM_View}
  n68 = Multiply(
   A=Add(
    A=MaterialFunctionCall(Float3=@n67) {materialFunction=&BreakOutFloat3Components}.G
    B=ScalarParameter() {defaultValue=-0.5, parameterName=Vertical Darkening Offset}
   )
   B=ScalarParameter() {defaultValue=0.0595, parameterName=Vertical Darkening Amount}
  )
  n69 = Multiply(A=OneMinus(None=SphereMask(A=@n63, B=@n64, Radius=@n66) {attenuationRadius=0.5, hardnessPercent=0}), B=OneMinus(None=Clamp(None=@n68)))
  n70 = Add(A=Multiply(A=Clamp(None=@n69), B=ScalarParameter() {defaultValue=3.3063, parameterName=GlowAmount}), B=@n69) {constB=0.1}
  n71 = VectorParameter() {parameterName=2D Glow color, defaultValue=<rgba 0.2393 0.1308 0.37 2>}
  n72 = VectorParameter() {parameterName=Light shaft Color 1, defaultValue=<rgba 5 0.2707 4.4883 1>}
  n73 = VectorParameter() {parameterName=Light shaft Color 2, defaultValue=<rgba 0.8597 0.905 0.8911 1>}
  n74 = Divide(
   A=@n64
   B=Multiply(
    A=MaterialFunctionCall() {materialFunction=&BluglowPositionScale}.Radius
    B=ScalarParameter() {defaultValue=2.7632, parameterName=Scale On Center Light Shafts}
   ) {constB=2}
  )
  n75 = Add(A=@n74) {constB=0.5}
  n76 = Add(
   A=MaterialFunctionCall() {materialFunction=&BluglowPositionScale}.Random
   B=Time() {bOverride_Period=true, period=10000000000}
  )
  n77 = Multiply(A=@n74) {constB=1.5}
  n78 = TextureSample(UVs=Rotator(Coordinate=Add(A=@n77) {constB=0.5}, Time=@n76) {speed=-0.1}) {texture=/Game/Effects/Fort_Effects/BluGlow/ZoomNoise}
  n79 = Multiply(A=TextureSample(UVs=Rotator(Coordinate=@n75, Time=@n76)) {texture=&ZoomNoise}.R, B=Add(A=@n78.G, B=@n78.G))
  n80 = ScalarParameter() {defaultValue=1, desc=For Blueprints, parameterName=LightShaftMultipler}
  n81 = MaterialFunctionCall!(
   Base=Multiply(A=Multiply(A=@n62, B=@n70), B=@n71.RGB)
   Add=Multiply(A=LinearInterpolate(A=@n72.RGB, B=@n73.RGB, Alpha=@n79), B=Multiply(A=@n79, B=@n80) {constB=2.233})
  ) {materialFunction=/Engine/Functions/Engine_MaterialFunctions02/Math/MultiplyAdd}
  n82 = Multiply(A=@n48, B=Add(A=Power(Base=@n81.Result, Exp=@n71.A), B=VectorParameter() {parameterName=Glow Pass, defaultValue=<rgba 0 0 0 0.1>}.RGB))
  n83 = ScalarParameter() {defaultValue=6.2637, parameterName=Metallic}
  n84 = OneMinus(None=@n49)
  n85 = Multiply(A=Multiply(A=@n82) {constB=0.5}, B=OneMinus(None=Fresnel() {baseReflectFraction=0, exponent=0.5}))
  n86 = Time() {bOverride_Period=true, period=99999997952}
  n87 = Constant2Vector() {g=0.0333, r=0.075}
  n88 = Desaturation(
   None=TextureSample(UVs=Multiply(A=@n86, B=@n87)) {samplerType=SAMPLERTYPE_Normal, texture=/Game/Global/Textures/Default/Noise/64x64Noise}.RGB
   Fraction=Constant() {r=0.5}
  )
  n89 = MaterialFunctionCall!(
   Material=MakeMaterialAttributes!(
    BaseColor=Multiply(A=@n84, B=@n82)
    Metallic=Multiply(A=@n49, B=@n83)
    Specular=ScalarParameter() {defaultValue=0.5, parameterName=Specular}
    Roughness=LinearInterpolate(A=@n46, B=@n47, Alpha=@n49)
    Anisotropy
    EmissiveColor=MaterialFunctionCall(Base=@n85, Add=@n88) {materialFunction=&MultiplyAdd}.Result
    Opacity=ScalarParameter() {defaultValue=0.3442, parameterName=Opacity}
    OpacityMask
    Normal=Transform(None=@n45)
    Tangent
    WorldPositionOffset
    SubsurfaceColor=@n82
    ClearCoat
    ClearCoatRoughness
    AmbientOcclusion
    Refraction
    CustomizedUVs_0
    CustomizedUVs_1
    CustomizedUVs_2
    CustomizedUVs_3
    CustomizedUVs_4
    CustomizedUVs_5
    CustomizedUVs_6
    CustomizedUVs_7
    PixelDepthOffset
    ShadingModel
    Displacement
   )
   WorldPositionOffset=Add(A=@n59, B=@n52)
  ) {desc=Add bob, materialFunction=/Engine/Functions/MaterialLayerFunctions/MatLayerBlend_AddWorldPositionOffset}
  n90 = GetMaterialAttributes(None=@n89.Blended Material) {attributeGetTypes [#1, #2]}
  n91 = GetMaterialAttributes(None=StaticSwitchParameter(
    True=MaterialFunctionCall!(Material=@n42.Result, Mask, WorldPositionOffset=@n90.World Position Offset) {materialFunction=/Engine/Functions/MaterialLayerFunctions/MatLayerBlend_OverrideWorldPositionOffset}.Blended Material
    False=@n89.Blended Material
   ) {parameterName=Hologram Shader?}) {attributeGetTypes [#2]}
  n92 = WorldPosition() {worldPositionShaderOffset=WPT_ExcludeAllShaderOffsets}
  n93 = MaterialFunctionCall!(World Position=Add(A=@n91.World Position Offset, B=@n92.XYZ), DistanceCullFadeAnimationOffset, Use actor center instead of pivot point) {desc=calc lod offset, materialFunction=/Game/Packages/Fortress_SharedMaterials/Base_Material_Functions/LODEffect}
  n94 = VectorParameter() {parameterName=WorldPostionOfBlackHole, defaultValue=<rgba -30000 25536 1952 0>}
  n95 = MaterialFunctionCall!(
   WorldPostionOfBlackHole=AppendVector(A=@n94.RGB, B=@n94.RGB)
   World Position=Add(A=Add(A=@n91.World Position Offset, B=@n93.Result) {desc=add to previous offset}, B=@n92.XYZ) {desc=find new world position}
  ) {desc=calc black hole, materialFunction=/Game/Packages/Fortnite_UniqueMaterials_1/BlackHoleFunction}
  n96 = BreakMaterialAttributes!(Attr=MaterialFunctionCall(Material=@n91.MaterialAttributes, WorldPositionOffset=Add(A=@n93.Result, B=@n95.Result) {desc=add black hole to lod offsets}) {desc=add to previous offsets, materialFunction=&MatLayerBlend_AddWorldPositionOffset}.Blended Material)
  Reroute(None=@n96.WorldPositionOffset)
  StaticSwitchParameter(True=Add(A=ScalarParameter() {defaultValue=1, parameterName=Chrome Brightness}, B=@n96.BaseColor), False=@n96.BaseColor) {parameterName=Chrome}
  StaticSwitchParameter(True=Add(A=ScalarParameter() {defaultValue=1, parameterName=Chrome Metalness}, B=@n96.Metallic), False=@n96.Metallic) {parameterName=Chrome}
 ]
}
